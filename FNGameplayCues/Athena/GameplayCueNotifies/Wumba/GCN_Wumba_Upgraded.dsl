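; Burst gameplay cue for a Wumba upgrade that tints its particle effect by level, in colors that resemble item rarity tiers.
; After the parent burst runs, it sets the first Cascade component's "Color" vector from a five-way select on the cue's gameplay effect level.
; A leftover-looking debug PrintString also outputs that level in red before the parent call.
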(fn OnBurst(Cascade)
  (params (MyTarget :type Actor%20Object%20Reference) (Parameters :type Gameplay%20Cue%20Parameters%20Structure%20%28by%20ref%29) (ParticleComponents :type Array%20of%20Cascade%20Particle%20System%20Component%20Object%20References) (AudioComponents :type Array%20of%20Audio%20Component%20Object%20References) (BurstCameraShakeInstance :type Camera%20Shake%20Base%20Object%20Reference) (BurstDecalInstance :type Decal%20Actor%20Object%20Reference))
  (bind (_normalizedmagnitude _rawmagnitude _effectcontext _matchedtagname _originaltag _aggregatedsourcetags _aggregatedtargettags _location _normal _instigator _effectcauser _sourceobject _physicalmaterial _gameplayeffectlevel) (Ability|GameplayCue|BreakGameplayCueParameters Parameters))
  (Development|PrintString 0 (Utilities|String|ToString(Integer) _gameplayeffectlevel) true true "(R=1.000000,G=0.000000,B=0.000000,A=1.000000)")
  (|CallParentFunction MyTarget Parameters ParticleComponents AudioComponents BurstCameraShakeInstance BurstDecalInstance)
  (Effects|Components|ParticleSystem|SetVectorParameter (Utilities|Array|Get(acopy) ParticleComponents) "Color" (Utilities|Select :Index _gameplayeffectlevel :Option%200 "0.150000,0.150000,0.150000" :Option%201 "0.050000,0.500000, 0" :Option%202 "0,0.400000,6.000000" :Option%203 "1.500000, 0,6.000000" :Option%204 "5.000000,1.500000, 0"))
  (return))
